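; Green room double-doors actor for Pilgrim Quickplay that presents the exit as a rift portal with sound, Niagara, camera shake and lens effects.
; After playlist data and the event router are ready, it listens for green room timer and quickplay state events.
; A timer change rumbles and bursts the portal only during Pregame; SongGameplay shuts effects down.
; A proximity collision volume adds and removes a cosmetic close-up lens effect aimed at the portal.
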(event Collision|EventActorBeginOverlap
  (params (OtherActor :type Actor%20Object%20Reference)))

(event EventBeginPlay
  (bind _event_router (Variables|Default|GetEvent%20Router))
  ;[1 1 1] Events
  (CallFunction|ActivateGreenroomPortalFX)
  (Class|FortAsyncActionCurrentPlaylistReady|PlaylistDataReadyAsync
    (:OnReady
      (bind _eventrouter (Pilgrim|PilgrimWaitForEventRouter 0)
        (:OnEventRouterAvailable
          (Variables|Default|SetEvent%20Router _eventrouter)
          (Utilities|FlowControl|DoOnce false
            (:Completed
              (Utilities|FlowControl|Sequence
                (:then_0
                  (bind (_asynctaskproxy _payload) (GameplayEventRouter|StartListeningToEvent _event_router "/PilgrimQuickplay/Items/PilgrimEvent_GreenRoomTimerStateChanged.PilgrimEvent_GreenRoomTimerStateChanged")
                    (:OnEventReceived
                      (bind _payload_greenroomtimerstate_2_04fc39c74228c50f7ae0f7b72d379be4 (Utilities|Struct|BreakPilgrimEventGreenRoomTimerStateChanged _payload))
                      (switch Utilities|FlowControl|Switch|SwitchonGreenRoomTimerState _payload_greenroomtimerstate_2_04fc39c74228c50f7ae0f7b72d379be4
                        (:NewEnumerator2
                          ;;[0.4775 0.9602 1] Ensure that we are in the Pregame state before firing off any FX (this prevents extra sounds when readying up on stage)
                          (bind _returnvalue (Class|PilgrimQuickplayPlayerStateComponent|GetPilgrimQuickplayPlayerStateComponent (Utilities|Array|Get(acopy) (Class|SparksMusicPlayspace|GetAllPlayerStates (Class|SparksMusicPlayspace|GetSparksMusicPlayspaceOfClass "/Script/PilgrimQuickplayRuntime.PilgrimQuickplayPlayspace")))))
                          (Utilities|IsValid _returnvalue
                            (:Is%20Valid
                              (Utilities|IsValid (Class|PilgrimQuickplayPlayerStateComponent|GetMainStateMachine _returnvalue)
                                (:Is%20Valid
                                  (if (== (GameplayStateMachine|GetActiveStateId (Class|PilgrimQuickplayPlayerStateComponent|GetMainStateMachine _returnvalue)) "(TagName=\"Pilgrim.Quickplay.PilgrimState.Id.Pregame\")")
                                    (Audio|SpawnSoundatLocation (Variables|Default|GetRiftSpawn) "14343.000000,-28449.000000,7088.000000")
                                    (CallFunction|StartRumble)
                                    (CallFunction|OpenDoors)))))))))))
                (:then_1
                  (bind (_asynctaskproxy_2 _payload_2) (GameplayEventRouter|StartListeningToEvent _event_router "/Script/PilgrimQPSharedRuntime.PilgrimEvent_QuickplayState_Began")
                    (:OnEventReceived
                      (bind _payload_stateid (Utilities|Struct|BreakPilgrimEventQuickplayStateBegan _payload_2))
                      (switch Utilities|FlowControl|Switch|SwitchonEPilgrimQuickplayState _payload_stateid
                        (:Pregame
                          (CallFunction|ActivateGreenroomPortalFX)
                          (Collision|SetCollisionEnabled (Variables|Default|GetPostProcessRadius) "QueryOnly"))
                        (:SongGameplay
                          ;;[1 1 1] Deactivate Portal Opening Rift FX
                          (Components|Activation|Deactivate (Variables|Default|GetCameraShakeSource_Portal_Ambient_Loop))
                          (Components|Activation|Deactivate (Variables|Default|GetPortalOpeningRiftFX))
                          (Components|Activation|Deactivate (Variables|Default|GetPortalBaseRiftFX))
                          (CallFunction|DeactivateCloseUpCamLensFX 0))))))
                (:then_2
                  (bind (_asynctaskproxy_3 _payload_3) (GameplayEventRouter|StartListeningToEvent _event_router "/Script/PilgrimQPSharedRuntime.PilgrimEvent_QuickplayState_Ended"))
                  (bind _payload_stateid_2 (Utilities|Struct|BreakPilgrimEventQuickplayStateEnded _payload_3))
                  (switch Utilities|FlowControl|Switch|SwitchonEPilgrimQuickplayState _payload_stateid_2
                    (:Pregame
                      (Collision|SetCollisionEnabled (Variables|Default|GetPostProcessRadius) "NoCollision"))))))))))))

(event Custom|StartRumble
  ;[1 1 1] Rumbling Doors
  (bind (_direction _portalcrackopening) (Timeline|Portal%20Opening PlayFromStart)
    (:Update
      (Niagara|SetNiagaraVariableByString(Float) (Variables|Default|GetPortalBaseRiftFX) "Open" _portalcrackopening))))

(event Custom|OpenDoors
  ;[1 1 1] Opening Doors
  (Variables|Default|SetContinueRumbeling false)
  (Utilities|FlowControl|Sequence
    (:then_0
      ;[1 1 1] Rumbling Doors
      (Timeline|Portal%20Opening Stop))
    (:then_1
      (Utilities|FlowControl|Sequence
        (:then_0
          (Utilities|FlowControl|Delay 0.23)
          ;;[0.0625 0.0625 0.0625] Burst Cam Shake
          (Camera|PlayWorldCameraShake "/Game/VisualThreatManager/Rifts/Blueprints/B_RiftImpact_CameraShake.B_RiftImpact_CameraShake_C" (Transformation|GetWorldLocation (Variables|Default|GetDefaultSceneRoot)) 1.5e+03 6e+03)
          ;;[0.002232 0 0.2292] Activate Rift FX
          (Audio|SpawnSoundatLocation (Variables|Default|GetEnterRift))
          (Components|Activation|Activate (Variables|Default|GetPortalOpeningRiftFX) false))
        (:then_1
          (Utilities|FlowControl|Delay 0.25)
          ;;[0.2541 0.3275 1] Apply Camera Lens Portal Burst FX
          (for _array_element (Fortnite|GetLocalFortPlayerControllers true)
            (Feedback|SpawnGenericCameraLensEffect (Game|Player|GetPlayerPawn _array_element) "/FNGameplayCues/Effects/Fort_Effects/Camera_Effects/B_CameraLens_Teleport.B_CameraLens_Teleport_C")))
        (:then_2
          (Utilities|FlowControl|Delay 0.25)
          (for _array_element_2 (Fortnite|GetLocalFortPlayerControllers true)
            (bind _returnvalue (Feedback|SpawnGenericCameraLensEffect (Game|Player|GetPlayerPawn _array_element_2) "/SparksEffects/Effects/Intro/Rift_Portal/BP/BP_Camera_Lens_PortalGreenRoom.BP_Camera_Lens_PortalGreenRoom_C"))
            (Effects|Components|ParticleSystem|SetVectorParameter (CameraLensEffect|GetPrimaryParticleComponent _returnvalue) "PortalPosition" (Transformation|GetActorLocation))))))))

(event OnComponentBeginOverlap(PostProcessRadius)
  (params (OverlappedComponent :type Primitive%20Component%20Object%20Reference) (OtherActor :type Actor%20Object%20Reference) (OtherComp :type Primitive%20Component%20Object%20Reference) (OtherBodyIndex :type Integer) (bFromSweep :type Boolean) (SweepResult :type Hit%20Result%20Structure%20%28by%20ref%29))
  ;[1 1 1] Close Up Portal Cam Lense FX
  (Networking|CanExecuteCosmeticEvents
    (:True
      (bind _asfort_player_pawn (Utilities|Casting|CastToFortPlayerPawn OtherActor))
      (bind _returnvalue (Feedback|SpawnGenericCameraLensEffect _asfort_player_pawn "/SparksEffects/Effects/Intro/Rift_Portal/BP/BP_Camera_Lens_PortalGreenRoom_CloseUp.BP_Camera_Lens_PortalGreenRoom_CloseUp_C"))
      (Variables|Default|SetCloseUpLensePortalFX _returnvalue)
      (Effects|Components|ParticleSystem|SetVectorParameter (CameraLensEffect|GetPrimaryParticleComponent _returnvalue) "PortalPosition" (Transformation|GetActorLocation))
      (Components|Activation|Activate (CameraLensEffect|GetPrimaryParticleComponent _returnvalue) false))))

(event OnComponentEndOverlap(PostProcessRadius)
  (params (OverlappedComponent :type Primitive%20Component%20Object%20Reference) (OtherActor :type Actor%20Object%20Reference) (OtherComp :type Primitive%20Component%20Object%20Reference) (OtherBodyIndex :type Integer))
  ;[1 1 1] Close Up Portal Cam Lense FX
  (CallFunction|DeactivateCloseUpCamLensFX OtherActor))

(fn ActivateGreenroomPortalFX
  ;[1 1 1] Activate Base Portal Visuals
  (Components|Activation|Activate (Variables|Default|GetPortalBaseRiftFX) false)
  (Components|Activation|Activate (Variables|Default|GetCameraShakeSource_Portal_Ambient_Loop) false)
  (Collision|SetCollisionEnabled (Variables|Default|GetPostProcessRadius) "QueryOnly"))

(fn DeactivateCloseUpCamLensFX
  (params (Object :type Object%20Reference))
  (bind _closeuplenseportalfx (Variables|Default|GetCloseUpLensePortalFX))
  (Networking|CanExecuteCosmeticEvents
    (:True
      (Utilities|IsValid Object
        (:Is%20Valid
          (bind _asfort_player_pawn (Utilities|Casting|CastToFortPlayerPawn Object))
          (Components|Activation|Deactivate (CameraLensEffect|GetPrimaryParticleComponent _closeuplenseportalfx))
          (Camera|RemoveGenericCameraLensEffect (Variables|/Script/Engine.PlayerController|GetPlayerCameraManager (Fortnite|GetFortPlayerControllerfromActor _asfort_player_pawn)) _closeuplenseportalfx))
        (:Is%20Not%20Valid
          (for _array_element (Collision|GetOverlappingActors (Variables|Default|GetPostProcessRadius) 0 :!class /Script/Engine.PrimitiveComponent)
            (Utilities|IsValid _array_element
              (:Is%20Valid
                (CallFunction|DeactivateCloseUpCamLensFX _array_element)))))))))

(fn ConstructionScript)
